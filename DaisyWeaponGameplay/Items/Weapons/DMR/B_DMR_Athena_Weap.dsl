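; Weapon Blueprint for the Athena DMR that adds scope recoil, camera shake and local projectile prediction.
; Local firing plays a FireShake timeline timed to the cached firing rate, driving scope offset amount and randomized recoil.
; Firing shakes the camera while targeting, repeating on a timer when Quartz gunfire is enabled.
; Non-server Tick interpolates scope camera X/Y offsets during targeting, and locally spawned DMR bullets are flagged as predicted.
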(event EventBeginPlay
  ;[1 1 1] Cache firing rate
  (|CallParentFunction)
  (bind _outrow (Fortnite|GetRangedWeaponStatsRow (Data|GetWeaponDataWeaponStatHandle)))
  (Variables|Default|SetFiring%20Rate (.FiringRate (Utilities|Struct|BreakFortRangedWeaponStats _outrow))))

(event EventOnPlayWeaponFireFX
  (params (bPersistentFire :type Boolean) (bSecondaryFire :type Boolean))
  (bind _fireshake (Variables|B_DMR_Athena_Weap|GetFireShake))
  ;[1 1 1] Recoil Scope Mesh Animation
  (|CallParentFunction bPersistentFire bSecondaryFire)
  (if (Pawn|IsLocallyControlled (Variables|Actor|GetInstigator))
    (Components|Timeline|SetTimelineLength _fireshake (/ 1.0 (Variables|Default|GetFiring%20Rate)))
    (Components|Timeline|SetLooping _fireshake false)
    (CallFunction|RecoilCamShake)
    (Variables|Default|SetFiring true)
    (if (Sounds|IsQuartzGunfireEnabled)
      (bind _returnvalue (Utilities|Time|SetTimerbyEvent (EventDispatchers|CreateEvent :!event Custom|RecoilCamShake) (/ 1.0 (Variables|Default|GetFiring%20Rate)) true))
      (Variables|Default|SetRecoilTimerHandle _returnvalue)
      (else
        _))
    (bind (_direction _newtrack_0) (Timeline|FireShake PlayFromStart)
      (:Update
        (Variables|Scope|SetScope%20Camera%20Offset%20Amount (Math|Float|Lerp 2e+01 19.0 _newtrack_0))
        (Variables|Default|SetXRecoilOffset (Math|Float|Lerp 0.0 (Math|Random|RandomFloatinRange -0.8 0.8) _newtrack_0))
        (Variables|Default|SetYRecoilOffset (Math|Float|Lerp 0.0 0.3 _newtrack_0)))
      (:Finished
        (if (Sounds|IsQuartzGunfireEnabled)
          _)))))

(event EventOnStopWeaponFireFX
  ;[1 1 1] Recoil Scope Mesh Animation
  (|CallParentFunction))

(event Custom|RecoilCamShake
  ;[1 1 1] Recoil Scope Mesh Animation
  (if (Camera|GetIsTargeting)
    (CameraShakes|StartCameraShake :self (Variables|/Script/Engine.PlayerController|GetPlayerCameraManager (Game|GetPlayerController 0)) :Scale 1.0 :PlaySpace "CameraLocal")))

(event EventOnSetTargeting
  (params (bNewIsTargeting :type Boolean))
  (|CallParentFunction bNewIsTargeting))

(event EventTick
  (params (DeltaSeconds :type Float%20%28single-precision%29))
  (if (not (Networking|IsServer))
    (bind _returnvalue (GameplayTags|HasMatchingGameplayTag(Message) (Pawn|GetControlledPawn (Game|GetPlayerController 0)) "(TagName=\"Gameplay.Status.IsTargeting\")"))
    (if _returnvalue
      (Utilities|FlowControl|Sequence
        (:then_0
          (Utilities|DoOnceThenDo false
            (:DoOnce
              (Variables|Default|SetPreviousRotation (Pawn|GetControlRotation (Game|GetPlayerController 0) :!class /Script/Engine.Controller)))
            (:ThenDo
              (Utilities|FlowControl|Sequence))))
        (:then_1
          ;[1 1 1] Tick to interpolate the Scope Mesh
          (bind _output_get (Variables|Default|SetYOffset (Math|Interpolation|FInterpTo (Variables|Default|GetYOffset) (Variables|Default|GetYOffsetInterpTarget) DeltaSeconds 5.0)))
          (Variables|Scope|SetScopeCameraOffsetY (Utilities|Operators|Add _output_get (Variables|Default|GetYADSOffset) (Variables|Default|GetYRecoilOffset)))
          (bind _output_get_2 (Variables|Default|SetXOffset (Math|Interpolation|FInterpTo (Variables|Default|GetXOffset) (Utilities|Operators|Add (Variables|Default|GetXOffsetInterpTarget) (Variables|Default|GetXStrafeOffset) (Variables|Default|GetXRecoilOffset)) DeltaSeconds 5.0)))
          (Variables|Scope|SetScopeCameraOffsetX (Utilities|Operators|Add _output_get_2 (Variables|Default|GetXADSOffset) 0)))))))

(event FortWeapon|EventOnSpawnProjectile
  (params (SpawnProjectile :type Fort%20Projectile%20Base%20Object%20Reference))
  (if (Pawn|IsLocallyControlled (Variables|Actor|GetInstigator))
    (bind _asb_prj_bullet_dmr (Utilities|Casting|CastToB_Prj_Bullet_DMR SpawnProjectile))
    (Variables|/DaisyWeaponGameplay/Items/Weapons/DMR/B_Prj_Bullet_DMR.B_Prj_Bullet_DMR_C|SetIsPredicted true _asb_prj_bullet_dmr)))

(fn ConstructionScript
  (|CallParentFunction))
