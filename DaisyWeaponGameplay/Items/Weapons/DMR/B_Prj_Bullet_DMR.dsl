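; DMR bullet projectile Blueprint that handles trail, impact and water effects.
; Clients drive the Niagara trail from the muzzle, play surface-mapped impact sounds and spawn fading decals on non-pawn hits off mobile.
; On stop, the predicted projectile sends the Ability.DMR.Fire gameplay event with the hit, presumably for damage.
; Overlap and trace water hits feed gated, do-once splash logic using water surface info; a delayed trail activation works around a mobile bug.
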(event EventBeginPlay
  (bind _ns_coresniper_bullet_trail (Variables|Default|GetNS_CoreSniper_Bullet_Trail))
  (bind _ns_coresniper_bullet_trail_2 (Variables|Default|GetNS_CoreSniper_Bullet_Trail))
  (|CallParentFunction)
  (Networking|CanExecuteCosmeticEvents
    (:True
      (bind _asfort_player_pawn (Utilities|Casting|CastToFortPlayerPawn (Variables|Actor|GetInstigator)))
      ;[1 1 1] Hacky fix for FORT-587865
      (if (Fortnite|IsMobilePlatform)
        (Components|Activation|Deactivate _ns_coresniper_bullet_trail_2)
        (Utilities|FlowControl|Delay 0.04)
        (Components|Activation|Activate _ns_coresniper_bullet_trail_2 false))
      (Niagara|SetNiagaraVariable(Vector3) _ns_coresniper_bullet_trail "BulletOrigin" (Effects|GetMuzzleLocation (Variables|/Script/FortniteGame.FortPawn|GetCurrentWeapon _asfort_player_pawn) -1))
      (Niagara|SetNiagaraVariable(Float) _ns_coresniper_bullet_trail "StartTime" (Utilities|Time|GetGameTimeinSeconds))
      (Niagara|SetNiagaraVariable(Float) _ns_coresniper_bullet_trail "WPODuration" 0.2))
    (:False
      (Variables|/Script/Engine.PrimitiveComponent|SetbGenerateOverlapEvents false (Variables|Collision|GetCapsuleComponent)) ; Overlaps only needed on clients for some FX interaction
    )))

(event EventOnStop
  (params (Hit :type Hit%20Result%20Structure%20%28by%20ref%29))
  (bind (_bblockinghit _binitialoverlap _time _distance _location _impactpoint _normal _impactnormal _physmat _hitactor) (Collision|BreakHitResult Hit))
  (bind _returnvalue (Game|GetInstigator))
  (bind (_asfort_pawn _bsuccess) (Utilities|Casting|CastToFortPawn _hitactor))
  (bind _output_get (Variables|Default|SetCachedHit Hit))
  (Utilities|FlowControl|Sequence
    (:then_0
      (Variables|/Script/Engine.ProjectileMovementComponent|SetProjectileGravityScale 0.0 (Variables|Movement|GetProjectileMovementComponent))
      (Class|BulletWhipTrackerComponentBase|Disable (Variables|Visual|GetBulletWhipTrackerComponent)))
    (:then_1
      (Networking|CanExecuteCosmeticEvents
        (:True
          (CallFunction|PlayImpactSFX (Variables|/Script/PhysicsCore.PhysicalMaterial|GetSurfaceType _physmat) _hitactor)
          (Transformation|DetachFromComponent (Variables|Default|GetNS_CoreSniper_Bullet_Trail) "KeepWorld" "KeepWorld" "KeepWorld" true)
          (Utilities|FlowControl|Sequence
            (:then_0
              (Niagara|SetNiagaraVariable(Bool) (Variables|Default|GetNS_CoreSniper_Bullet_Trail) "Deactive" true)
              (Utilities|FlowControl|Delay 0.2)
              (Components|Activation|Deactivate (Variables|Default|GetNS_CoreSniper_Bullet_Trail)))
            (:then_1
              (if (not (or (Fortnite|IsMobilePlatform) _bsuccess))
                (bind _returnvalue_2 (Rendering|Decal|SpawnDecalatLocation "/Game/Weapons/Materials/MD_Decal_BulletHole_Lifetime_Inst.MD_Decal_BulletHole_Lifetime_Inst" (Variables|Default|GetDecalSize) _location (Math|Conversions|RotationFromXVector (* _impactnormal -1.0)) 4.5))
                (Rendering|Components|Decal|SetFadeOut _returnvalue_2 0.0 4.5 true)))))))
    (:then_2
      (if (Variables|Default|GetIsPredicted)
        (bind _asfort_player_pawn (Utilities|Casting|CastToFortPlayerPawn _returnvalue))
        (Variables|Default|SetDamageContext (Class|AthenaFunctionLibrary|MakeSimpleFortEffectContext _returnvalue))
        (Ability|EffectContext|AddHitResult (Variables|Default|GetDamageContext) _output_get)
        (Ability|SendGameplayEventtoActor _asfort_player_pawn "(TagName=\"Ability.DMR.Fire\")" (Utilities|Struct|MakeGameplayEventData "(TagName=\"Ability.DMR.Fire\")" _asfort_player_pawn 0 0 0 (Variables|Default|GetDamageContext) "(GameplayTags=)" "(GameplayTags=)" 0.0 (Ability|TargetData|AbilityTargetDatafromActorArray (Utilities|Array|MakeArray (.HitActor (Collision|BreakHitResult _output_get))))))))))

(event EventOnTouched
  (params (OtherActor :type Actor%20Object%20Reference) (OtherComp :type Primitive%20Component%20Object%20Reference) (HitResult :type Hit%20Result%20Structure%20%28by%20ref%29) (bIsOverlap :type Boolean))
  (bind (_bblockinghit _binitialoverlap _time _distance _location _impactpoint _normal) (Collision|BreakHitResult (Variables|Default|GetWaterHit)))
  (bind (_bblockinghit_2 _binitialoverlap_2 _time_2 _distance_2 _location_2) (Collision|BreakHitResult (Variables|Default|GetWaterHit)))
  ;[0 0.298 1] Water Impact Projectile
  (Networking|CanExecuteCosmeticEvents
    (:True
      (Utilities|FlowControl|Gate false
        (:Exit
          (if bIsOverlap
            (Variables|Default|SetWaterHit HitResult)
            (bind _asfort_water_body_actor (Utilities|Casting|CastToFortWaterBodyActor OtherActor))
            (Utilities|FlowControl|Sequence
              (:then_0
                (bind _outwatersurfaceinfo (WaterBody|GetWaterSurfaceInfoStruct _asfort_water_body_actor _location_2 true false false))
                (Variables|Default|SetWaterDepth (.WaterDepth (Utilities|Struct|BreakWaterSurfaceInfo _outwatersurfaceinfo)))
                (Variables|Default|SetWaterIndex (.WaterBodyIdx (Utilities|Struct|BreakWaterSurfaceInfo _outwatersurfaceinfo)))
                (Variables|Default|SetTraceImpactNormal (Utilities|Struct|BreakWaterSurfaceInfo _outwatersurfaceinfo))
                (Utilities|FlowControl|DoOnce false
                  (:Completed
                    (CallFunction|PlayImpactSFX "SurfaceType8")
                    (bind _returnvalue (Niagara|SpawnSystematLocation (Variables|Default|GetNiagaraWaterImpactFX) _location (Math|Conversions|RotationFromXVector _normal)))
                    (Effects|Components|ParticleSystem|SetFloatParameter _returnvalue "WaterDepth" (Variables|Default|GetWaterDepth))
                    (Effects|Components|ParticleSystem|SetFloatParameter _returnvalue "WaterBodyIndex" (Math|Conversions|ToFloat(Integer) (Variables|Default|GetWaterIndex))))))
              (:then_1
                (Niagara|SetNiagaraVariable(Bool) (Variables|Default|GetNS_CoreSniper_Bullet_Trail) "Deactive" true)))))) ; Block this if the water trace hit
    )))

(event EventOnSniperTraceWaterHit
  (params (WaterHit :type Hit%20Result%20Structure%20%28by%20ref%29) (WaterBody :type Fort%20Water%20Body%20Actor%20Object%20Reference))
  (bind (_bblockinghit _binitialoverlap _time _distance _location _impactpoint _normal) (Collision|BreakHitResult (Variables|Default|GetWaterHit)))
  (bind (_bblockinghit_2 _binitialoverlap_2 _time_2 _distance_2 _location_2) (Collision|BreakHitResult (Variables|Default|GetWaterHit)))
  ;[0 0.298 1] Water Impact Trace
  (Utilities|FlowControl|DoOnce false
    (:Completed
      (Utilities|IsValid WaterBody
        (:Is%20Valid
          (bind _output_get (Variables|Default|SetWaterHit WaterHit))
          (Utilities|FlowControl|Sequence
            (:then_0
              ;[0 0.298 1] Water Impact Projectile
              (Utilities|FlowControl|Gate false
                (:Exit
                  (if (stale EventOnTouched bIsOverlap)
                    (Variables|Default|SetWaterHit (stale EventOnTouched HitResult))
                    (bind _asfort_water_body_actor (Utilities|Casting|CastToFortWaterBodyActor (stale EventOnTouched OtherActor)))
                    (Utilities|FlowControl|Sequence
                      (:then_0
                        (bind _outwatersurfaceinfo (WaterBody|GetWaterSurfaceInfoStruct _asfort_water_body_actor _location_2 true false false))
                        (Variables|Default|SetWaterDepth (.WaterDepth (Utilities|Struct|BreakWaterSurfaceInfo _outwatersurfaceinfo)))
                        (Variables|Default|SetWaterIndex (.WaterBodyIdx (Utilities|Struct|BreakWaterSurfaceInfo _outwatersurfaceinfo)))
                        (Variables|Default|SetTraceImpactNormal (Utilities|Struct|BreakWaterSurfaceInfo _outwatersurfaceinfo)))
                      (:then_1
                        (Niagara|SetNiagaraVariable(Bool) (Variables|Default|GetNS_CoreSniper_Bullet_Trail) "Deactive" true :!label _label)))))) ; Block this if the water trace hit
            )
            (:then_1
              (Utilities|FlowControl|Delay 0.01)
              (bind _outwatersurfaceinfo_2 (WaterBody|GetWaterSurfaceInfoStruct WaterBody (.Location (Collision|BreakHitResult _output_get)) true false false))
              (Variables|Default|SetWaterDepth (.WaterDepth (Utilities|Struct|BreakWaterSurfaceInfo _outwatersurfaceinfo_2)))
              (Variables|Default|SetWaterIndex (.WaterBodyIdx (Utilities|Struct|BreakWaterSurfaceInfo _outwatersurfaceinfo_2))))
            (:then_2
              (goto _label)))
          (Utilities|FlowControl|DoOnce false
            (:Completed
              (CallFunction|PlayImpactSFX "SurfaceType8")
              (bind _returnvalue (Niagara|SpawnSystematLocation (Variables|Default|GetNiagaraWaterImpactFX) _location (Math|Conversions|RotationFromXVector _normal)))
              (Effects|Components|ParticleSystem|SetFloatParameter _returnvalue "WaterDepth" (Variables|Default|GetWaterDepth))
              (Effects|Components|ParticleSystem|SetFloatParameter _returnvalue "WaterBodyIndex" (Math|Conversions|ToFloat(Integer) (Variables|Default|GetWaterIndex))))))))))

(fn GetSurfaceTypeFromInfo
  (params (Surface%20Type :type EPhysicalSurface%20Enum :default "SurfaceType_Default") (Actor :type Actor%20Object%20Reference))
  (bind (_asbuilding_static_mesh _bsuccess) (Utilities|Casting|CastToBuildingSMActor Actor))
  (Development|PrintString (Utilities|String|Append "SurfaceInfo: " (Utilities|String|EnumtoString Surface%20Type) ", " (Utilities|GetDisplayName Actor)))
  (if _bsuccess
    (return (Utilities|Select :Index (Variables|/Script/FortniteGame.BuildingSMActor|GetResourceType _asbuilding_static_mesh) :Wood "SurfaceType1" :Stone "SurfaceType2" :Metal "SurfaceType3" :Permanite "SurfaceType_Default" :GoldCurrency "SurfaceType_Default" :Ingredient "SurfaceType_Default" :None "SurfaceType_Default"))
    (else
      (return Surface%20Type))))

(fn PlayImpactSFX
  (params (Surface%20Type :type EPhysicalSurface%20Enum :default "SurfaceType_Default") (Actor :type Actor%20Object%20Reference))
  (if (Utilities|Enum|Equal(Enum) Surface%20Type "SurfaceType8")
    (if (not (Variables|Default|GetbHasPlayedWaterSound))
      (Variables|Default|SetbHasPlayedWaterSound true)))
  (Audio|PlaySoundatLocation (select (Utilities|IsValid (Utilities|Map|Find (Variables|Default|GetSurfaceImpactSFX) (CallFunction|GetSurfaceTypefromInfo Surface%20Type Actor)) :!class /Script/Engine.KismetSystemLibrary) (Utilities|Map|Find (Variables|Default|GetSurfaceImpactSFX) (CallFunction|GetSurfaceTypefromInfo Surface%20Type Actor)) (Utilities|Map|Find (Variables|Default|GetSurfaceImpactSFX) "SurfaceType_Default")) (Transformation|GetActorLocation)))

(fn ConstructionScript
  (Projectile|SetGravityScale (ScalableFloat|GetValueatLevel (Variables|Default|GetProjectileGravityScale))))
