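# Sound cue for the DMR weapon drop sound in the DaisyWeaponGameplay plugin.
# A single DMR_Drop wave player feeds a Modulator left at default settings, with a boosted cue volume.
# It uses the shared weapon drop attenuation, the loot collect sound class and the Weapon_Drop_StopOldest concurrency set to limit overlapping drops.
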SoundCue /DaisyWeaponGameplay/Items/Weapons/DMR/Sounds/DMR_Drop_Cue
root=Modulator_0
nodes [
 {id=Modulator_0, children [WavePlayer_0]}
 {id=WavePlayer_0, properties {soundWaveAssetPtr=&/DMR_Drop}}
]
properties {
 attenuationSettings=&Fort_Weapon_Drop_Att
 maxDistance=3000
 soundClassObject=&Fort_Loot_Collect
 volumeMultiplier=1.5
 concurrencyOverrides {resolutionRule=PreventNew}
 concurrencySet [&Weapon_Drop_StopOldest]
}
