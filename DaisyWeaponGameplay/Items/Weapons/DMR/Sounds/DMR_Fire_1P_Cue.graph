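# Sound cue for the DMR's first-person gunshot, mixing four randomized layers: shot, tail, mechanism and shell casings.
# The shot plays immediately; delayed mechanism, tail and casing layers follow, each with its own Modulator, with pitch variation only on the mechanism.
# A SoundClass node at the root overrides to the third-person fire class, although the cue's own class is first-person.
# It uses first-person weapon attenuation and gunfire concurrency, with debug enabled.
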SoundCue /DaisyWeaponGameplay/Items/Weapons/DMR/Sounds/DMR_Fire_1P_Cue
root=SoundClass_3
nodes [
 {id=Delay_0, children [Modulator_0], properties {delayMax=0.4, delayMin=0.4}}
 {id=Delay_1, children [Random_3], properties {delayMax=0.25, delayMin=0.25}}
 {id=Delay_2, children [Modulator_1], properties {delayMax=0.03, delayMin=0.03}}
 {id=Mixer_0, children [Modulator_3,Modulator_2,Delay_2,Delay_0], properties {inputVolume [1,1,1,1]}}
 {id=Modulator_0, children [Random_1], properties {pitchMax=1, pitchMin=1, volumeMax=0.05, volumeMin=0.05}}
 {id=Modulator_1, children [Random_2], properties {pitchMax=1.1, pitchMin=1.2, volumeMax=1, volumeMin=1}}
 {id=Modulator_2, children [Delay_1], properties {pitchMax=1, pitchMin=1, volumeMax=0.4, volumeMin=0.4}}
 {id=Modulator_3, children [Random_0], properties {pitchMax=1, pitchMin=1, volumeMax=1.2, volumeMin=1.2}}
 {id=Random_0, children [WavePlayer_3,WavePlayer_4,WavePlayer_5,WavePlayer_0,WavePlayer_1], properties {weights [1,1,1,1,1]}}
 {
  id=Random_1
  children [WavePlayer_37, WavePlayer_38, WavePlayer_39, WavePlayer_40, WavePlayer_41, WavePlayer_42, WavePlayer_43]
  properties {weights [1,1,1,1,1,1,1]}
 }
 {id=Random_2, children [WavePlayer_7, WavePlayer_8, WavePlayer_9, WavePlayer_10, WavePlayer_11], properties {weights [1,1,1,1,1]}}
 {
  id=Random_3
  children [WavePlayer_18, WavePlayer_19, WavePlayer_20, WavePlayer_21, WavePlayer_22, WavePlayer_23]
  properties {weights [1,1,1,1,1,1]}
 }
 {id=SoundClass_3, children [Mixer_0], properties {soundClassOverride=&Fort_WeaponRanged_Fire_3P_PSM}}
 {id=WavePlayer_0, properties {soundWaveAssetPtr=&/DMR_Shoot_05}}
 {id=WavePlayer_1, properties {soundWaveAssetPtr=&/DMR_Shoot_06}}
 {id=WavePlayer_10, properties {soundWaveAssetPtr=&/DMR_Mech_04}}
 {id=WavePlayer_11, properties {soundWaveAssetPtr=&/DMR_Mech_05}}
 {id=WavePlayer_18, properties {soundWaveAssetPtr=&/DMR_Shoot_Tail_01}}
 {id=WavePlayer_19, properties {soundWaveAssetPtr=&/DMR_Shoot_Tail_02}}
 {id=WavePlayer_20, properties {soundWaveAssetPtr=&/DMR_Shoot_Tail_03}}
 {id=WavePlayer_21, properties {soundWaveAssetPtr=&/DMR_Shoot_Tail_04}}
 {id=WavePlayer_22, properties {soundWaveAssetPtr=&/DMR_Shoot_Tail_05}}
 {id=WavePlayer_23, properties {soundWaveAssetPtr=&/DMR_Shoot_Tail_06}}
 {id=WavePlayer_3, properties {soundWaveAssetPtr=&/DMR_Shoot_01}}
 {id=WavePlayer_37, properties {soundWaveAssetPtr=&/DMR_ShellCasings_01}}
 {id=WavePlayer_38, properties {soundWaveAssetPtr=&/DMR_ShellCasings_02}}
 {id=WavePlayer_39, properties {soundWaveAssetPtr=&/DMR_ShellCasings_03}}
 {id=WavePlayer_4, properties {soundWaveAssetPtr=&/DMR_Shoot_02}}
 {id=WavePlayer_40, properties {soundWaveAssetPtr=&/DMR_ShellCasings_04}}
 {id=WavePlayer_41, properties {soundWaveAssetPtr=&/DMR_ShellCasings_05}}
 {id=WavePlayer_42, properties {soundWaveAssetPtr=&/DMR_ShellCasings_06}}
 {id=WavePlayer_43, properties {soundWaveAssetPtr=&/DMR_ShellCasings_07}}
 {id=WavePlayer_5, properties {soundWaveAssetPtr=&/DMR_Shoot_03}}
 {id=WavePlayer_7, properties {soundWaveAssetPtr=&/DMR_Mech_01}}
 {id=WavePlayer_8, properties {soundWaveAssetPtr=&/DMR_Mech_02}}
 {id=WavePlayer_9, properties {soundWaveAssetPtr=&/DMR_Mech_03}}
]
properties {
 attenuationSettings=&Fort_Weapon_Projectile_FirstPerson_Athena_Att
 bDebug=true
 maxDistance=2097152
 soundClassObject=&Fort_WeaponRanged_Fire_1P_PSM
 volumeMultiplier=0.7
 attenuationOverrides {bSpatialize=false}
 concurrencyOverrides {bLimitToOwner=true, maxCount=3, resolutionRule=StopOldest}
 concurrencySet [&Weapon_Gunfire_FirstPerson_Conc]
}
