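# Sound cue for the DMR's third-person gunfire, layering close, medium and distant shot sounds.
# Close mixes random shot and tail waves; each range layer has its own modulation and attenuation, with sound class overrides separating close from medium and distant.
# A quality-level node switches the distant layer between pools of six and three random waves.
# Global and local gunfire concurrency sets are assigned.
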SoundCue /DaisyWeaponGameplay/Items/Weapons/DMR/Sounds/DMR_Fire_3P_Cue
root=Mixer_2
nodes [
 {id=Attenuation_3, children [Modulator_0], properties {attenuationSettings=&Fort_Weapon_Projectile_Close_Athena_Att}}
 {
  id=Attenuation_4
  children [Modulator_2]
  properties {attenuationSettings=&Fort_Weapon_Projectile_Distant_Mono_Athena_Att}
 }
 {
  id=Attenuation_5
  children [Modulator_4]
  properties {attenuationSettings=&Fort_Weapon_Projectile_Medium_Mono_Athena_Att}
 }
 {id=Mixer_0, children [Random_2,Random_4], properties {inputVolume [1,0.7]}}
 {id=Mixer_2, children [Modulator_3,SoundClass_4], properties {inputVolume [1,1]}}
 {id=Mixer_3, children [SoundClass_3,SoundClass_5], properties {inputVolume [1,1]}}
 {id=Modulator_0, children [Mixer_0], properties {pitchMax=1.3, pitchMin=1.1, volumeMax=1}}
 {id=Modulator_2, children [QualityLevel_0], properties {pitchMax=0.95, pitchMin=0.9, volumeMax=0.6, volumeMin=0.55}}
 {id=Modulator_3, children [Mixer_3], properties {volumeMax=1, volumeMin=1}}
 {id=Modulator_4, children [Random_1], properties {pitchMax=1, volumeMax=0.5, volumeMin=0.45}}
 {id=QualityLevel_0, children [Random_0,Random_0,Random_3,Random_3]}
 {
  id=Random_0
  children [WavePlayer_19, WavePlayer_20, WavePlayer_21, WavePlayer_22, WavePlayer_23, WavePlayer_24]
  properties {weights [1,1,1,1,1,1]}
 }
 {
  id=Random_1
  children [WavePlayer_6, WavePlayer_7, WavePlayer_15, WavePlayer_16, WavePlayer_17, WavePlayer_18]
  properties {weights [1,1,1,1,1,1]}
 }
 {id=Random_2, children [WavePlayer_0,WavePlayer_1,WavePlayer_2,WavePlayer_3,WavePlayer_4,WavePlayer_5], properties {weights [1,1,1,1,1,1]}}
 {id=Random_3, children [WavePlayer_22, WavePlayer_23, WavePlayer_24], properties {weights [1,1,1]}}
 {
  id=Random_4
  children [WavePlayer_8, WavePlayer_9, WavePlayer_10, WavePlayer_11, WavePlayer_12, WavePlayer_13]
  properties {weights [1,1,1,1,1,1]}
 }
 {
  id=SoundClass_3
  children [Attenuation_3]
  properties {soundClassOverride=&Fort_WeaponRanged_Fire_3P_PSM}
 }
 {
  id=SoundClass_4
  children [Attenuation_4]
  properties {soundClassOverride=&Fort_WeaponRanged_Fire_Distant}
 }
 {
  id=SoundClass_5
  children [Attenuation_5]
  properties {soundClassOverride=&Fort_WeaponRanged_Fire_Distant}
 }
 {id=WavePlayer_0, properties {soundWaveAssetPtr=&/DMR_Shoot_Close_01}}
 {id=WavePlayer_1, properties {soundWaveAssetPtr=&/DMR_Shoot_Close_02}}
 {id=WavePlayer_10, properties {soundWaveAssetPtr=&/DMR_Shoot_Tail_03}}
 {id=WavePlayer_11, properties {soundWaveAssetPtr=&/DMR_Shoot_Tail_04}}
 {id=WavePlayer_12, properties {soundWaveAssetPtr=&/DMR_Shoot_Tail_05}}
 {id=WavePlayer_13, properties {soundWaveAssetPtr=&/DMR_Shoot_Tail_06}}
 {id=WavePlayer_15, properties {soundWaveAssetPtr=&/DMR_Shoot_Medium_03}}
 {id=WavePlayer_16, properties {soundWaveAssetPtr=&/DMR_Shoot_Medium_04}}
 {id=WavePlayer_17, properties {soundWaveAssetPtr=&/DMR_Shoot_Medium_05}}
 {id=WavePlayer_18, properties {soundWaveAssetPtr=&/DMR_Shoot_Medium_06}}
 {id=WavePlayer_19, properties {soundWaveAssetPtr=&/DMR_Shoot_Distant_01}}
 {id=WavePlayer_2, properties {soundWaveAssetPtr=&/DMR_Shoot_Close_03}}
 {id=WavePlayer_20, properties {soundWaveAssetPtr=&/DMR_Shoot_Distant_02}}
 {id=WavePlayer_21, properties {soundWaveAssetPtr=&/DMR_Shoot_Distant_03}}
 {id=WavePlayer_22, properties {soundWaveAssetPtr=&/DMR_Shoot_Distant_04}}
 {id=WavePlayer_23, properties {soundWaveAssetPtr=&/DMR_Shoot_Distant_05}}
 {id=WavePlayer_24, properties {soundWaveAssetPtr=&/DMR_Shoot_Distant_06}}
 {id=WavePlayer_3, properties {soundWaveAssetPtr=&/DMR_Shoot_Close_04}}
 {id=WavePlayer_4, properties {soundWaveAssetPtr=&/DMR_Shoot_Close_05}}
 {id=WavePlayer_5, properties {soundWaveAssetPtr=&/DMR_Shoot_Close_06}}
 {id=WavePlayer_6, properties {soundWaveAssetPtr=&/DMR_Shoot_Medium_01}}
 {id=WavePlayer_7, properties {soundWaveAssetPtr=&/DMR_Shoot_Medium_02}}
 {id=WavePlayer_8, properties {soundWaveAssetPtr=&/DMR_Shoot_Tail_01}}
 {id=WavePlayer_9, properties {soundWaveAssetPtr=&/DMR_Shoot_Tail_02}}
]
properties {
 bDebug=true
 maxDistance=27500
 soundClassObject=&Fort_WeaponRanged_Fire_3P_PSM
 volumeMultiplier=1
 concurrencyOverrides {bLimitToOwner=true, maxCount=3, resolutionRule=StopOldest}
 concurrencySet [&Weapon_Gunfire_Global_Conc, &Weapon_Gunfire_Local_Conc]
}
