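# Sound cue for the DMR's low-ammo warning, built on a shared Fortnite high rifle low-ammo sound wave.
# A single wave player feeds a Modulator that pins volume and leaves pitch at its default random variation.
# It uses first-person weapon projectile attenuation and a ranged-weapon low-ammo sound class.
# A concurrency override refuses new instances once the limit is reached instead of cutting off playing ones.
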SoundCue /DaisyWeaponGameplay/Items/Weapons/DMR/Sounds/DMR_LowAmmo_Cue
root=Modulator_0
nodes [
 {id=Modulator_0, children [WavePlayer_34], properties {volumeMax=1, volumeMin=1}}
 {id=WavePlayer_34, properties {soundWaveAssetPtr=/Game/Sounds/Fort_Weapons/LowAmmo/weapon_low_ammo_hi_rifle_05}}
]
properties {
 attenuationSettings=&Fort_Weapon_Projectile_FirstPerson_Athena_Att
 bOverrideConcurrency=true
 maxDistance=2097152
 soundClassObject=/Game/Sounds/FortSoundClassesAndModes/Fort_WeaponRanged_LowAmmo_Degraded_PSM
 concurrencyOverrides {resolutionRule=PreventNew}
}
