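# Sound cue for the DMR scope zoom-in, played as weapon foley.
# A mixer layers a boosted scope-in wave with a quieter movement layer, randomly picked from four player run foley clips.
# Each layer passes through its own modulator to set gain, and the cue lowers the overall volume.
# The cue overrides attenuation for short-range falloff and uses a concurrency rule that blocks new instances at the limit.
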SoundCue /DaisyWeaponGameplay/Items/Weapons/DMR/Sounds/DMR_Scope_ZoomIn_Cue
root=Mixer_0
nodes [
 {id=Mixer_0, children [Modulator_0,Modulator_1], properties {inputVolume [0.4,1]}}
 {id=Modulator_0, children [Random_4], properties {volumeMax=0.8, volumeMin=0.8}}
 {id=Modulator_1, children [WavePlayer_3], properties {pitchMax=1, volumeMax=2.5, volumeMin=2.5}}
 {id=Random_4, children [WavePlayer_19, WavePlayer_1565, WavePlayer_1563, WavePlayer_1562], properties {weights [1,1,1,1]}}
 {id=WavePlayer_1562, properties {soundWaveAssetPtr=/Game/Sounds/Fort_Foley_Sounds/Moves/player_run_foley_12}}
 {id=WavePlayer_1563, properties {soundWaveAssetPtr=/Game/Sounds/Fort_Foley_Sounds/Moves/player_run_foley_10}}
 {id=WavePlayer_1565, properties {soundWaveAssetPtr=/Game/Sounds/Fort_Foley_Sounds/Moves/player_run_foley_03}}
 {id=WavePlayer_19, properties {soundWaveAssetPtr=/Game/Sounds/Fort_Foley_Sounds/Moves/player_run_foley_09}}
 {id=WavePlayer_3, properties {soundWaveAssetPtr=&/DMR_Scope_In}}
]
properties {
 attenuationSettings=&Fort_Weapon_Projectile_FirstPerson_Athena_Att
 bOverrideAttenuation=true
 bOverrideConcurrency=true
 maxDistance=2700
 pitchMultiplier=1.02
 soundClassObject=&Fort_WeaponFoley
 volumeMultiplier=0.251
 attenuationOverrides {falloffDistance=1500, nonSpatializedRadiusStart=200, attenuationShapeExtents {x=1200}}
 concurrencyOverrides {resolutionRule=PreventNew}
}
