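# Sound cue for the DMR scope zoom-out, routed through the weapon foley sound class.
# A mixer layers a volume-boosted scope-out wave with one of two randomly chosen player-run foley waves, reused here with slight upward pitch variation.
# The cue plays at a heavily reduced master volume, with inline attenuation overriding the assigned first-person projectile asset and a concurrency override that blocks new instances at the limit.
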SoundCue /DaisyWeaponGameplay/Items/Weapons/DMR/Sounds/DMR_Scope_ZoomOut_Cue
root=Mixer_0
nodes [
 {id=Mixer_0, children [Modulator_0,Modulator_1], properties {inputVolume [1,1]}}
 {id=Modulator_0, children [Random_2], properties {pitchMax=1.2, pitchMin=1}}
 {id=Modulator_1, children [WavePlayer_1], properties {pitchMax=1, pitchMin=1, volumeMax=3, volumeMin=3}}
 {id=Random_2, children [WavePlayer_7,WavePlayer_6], properties {weights [1,1]}}
 {id=WavePlayer_1, properties {soundWaveAssetPtr=&/DMR_Scope_Out}}
 {id=WavePlayer_6, properties {soundWaveAssetPtr=/Game/Sounds/Fort_Foley_Sounds/Moves/player_run_foley_05}}
 {id=WavePlayer_7, properties {soundWaveAssetPtr=/Game/Sounds/Fort_Foley_Sounds/Moves/player_run_foley_01}}
]
properties {
 attenuationSettings=&Fort_Weapon_Projectile_FirstPerson_Athena_Att
 bOverrideAttenuation=true
 bOverrideConcurrency=true
 maxDistance=2700
 soundClassObject=&Fort_WeaponFoley
 volumeMultiplier=0.05
 attenuationOverrides {falloffDistance=1500, nonSpatializedRadiusStart=200, attenuationShapeExtents {x=1200}}
 concurrencyOverrides {resolutionRule=PreventNew}
}
