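# Sound cue for the DMR weapon's stop sound, built on a reused auto rifle stop wave.
# A single wave player feeds a Modulator at the root.
# It uses long-range assault rifle attenuation and the third-person ranged weapon fire sound class.
# Local and global gunfire concurrency sets govern overlapping plays, and the cue's debug flag is left on.
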SoundCue /DaisyWeaponGameplay/Items/Weapons/DMR/Sounds/DMR_Stop_Cue
root=Modulator_1
nodes [{id=Modulator_1, children [WavePlayer_4]}, {id=WavePlayer_4, properties {soundWaveAssetPtr=/Game/Sounds/Fort_Weapons/AutoRifle/ar4_stop_01}}]
properties {
 attenuationSettings=/Game/Sounds/Attenuation/Fort_Weapon_AR_Lng_Att
 bDebug=true
 maxDistance=4000
 soundClassObject=&Fort_WeaponRanged_Fire_3P_PSM
 concurrencyOverrides {maxCount=2}
 concurrencySet [&Weapon_Gunfire_Local_Conc, &Weapon_Gunfire_Global_Conc]
}
